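; Function library of stateless helpers for Fortnite motorcycles, covering cosmetic feedback, part damage and impact responses.
; Callers pass in state, components and assets; helpers drive Niagara road/exhaust parameters, gameplay cues, camera shakes and force feedback.
; Others pop tires, spawn explosion wrecks, launch hit pawns with fall-instigator credit, and apply rim-scrape and ice gameplay effects.
; Tire smoke still prints a debug string, and the ice-effect removal branch is unreachable.
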(fn ApplyBoostGC
  (params (ApplyBoostGC :type Boolean) (DriverPawn :type Fort%20Player%20Pawn%20Object%20Reference) (BoostGC :type Gameplay%20Tag%20Structure))
  (bind _returnvalue (Ability|GameplayCue|MakeGameplayCueParameters 0.0))
  (bind _driverpawn (Variables|GetDriverPawn))
  (bind _boostgc (Variables|GetBoostGC))
  (Utilities|IsValid _driverpawn
    (:Is%20Valid
      (if (Pawn|IsLocallyViewed _driverpawn)
        (if ApplyBoostGC
          (Projectile|AddGameplayCue%28Looping%29Local _driverpawn _boostgc _returnvalue)
          (else
            (Projectile|RemoveGameplayCue%28Looping%29Local _driverpawn _boostgc _returnvalue)))))))

(fn DamageablePartChange
  (params (PoppedTireFX :type Niagara%20System%20Object%20Reference) (VehicleSkeletalMesh :type Scene%20Component%20Object%20Reference) (VehicleDamageablePartConfig :type Vehicle%20Damageable%20Part%20Config%20Structure%20%28by%20ref%29) (NewlyDisabled :type Boolean) (WheelSocketF :type Name) (WheelSocketB :type Name) (WindowTransformFrW :type Transform%20%28by%20ref%29) (RoadInteractionFX :type Niagara%20Particle%20System%20Component%20Object%20Reference) (WindowBreakFX :type Niagara%20System%20Object%20Reference) (VehicleSelfReference :type Fort%20Athena%20Vehicle%20Object%20Reference))
  (bind _vehicledamageablepartconfig (Variables|GetVehicleDamageablePartConfig))
  (bind (_shapename _bonename) (Utilities|Struct|BreakVehicleDamageablePartConfig _vehicledamageablepartconfig))
  (bind _windowtransformfrw (Variables|GetWindowTransformFrW))
  (bind _poppedtirefx (Variables|GetPoppedTireFX))
  (bind _vehicleskeletalmesh (Variables|GetVehicleSkeletalMesh))
  (bind _bonenamelocal (LocalVariables|GetBoneNameLocal))
  (bind _vehicleselfreference (Variables|GetVehicleSelfReference))
  (bind _roadinteractionfx (Variables|GetRoadInteractionFX))
  (if (Variables|GetNewlyDisabled)
    (LocalVariables|SetBoneNameLocal _bonename)
    (switch name _shapename
      (:wheel_f
        (Niagara|SpawnSystemAttached _poppedtirefx _vehicleskeletalmesh "None" (Transformation|GetSocketLocation _vehicleskeletalmesh (Variables|GetWheelSocketF)) "0, 0, 0" "KeepWorldPosition" false true "AutoRelease")
        (Vehicle|Tires|OnPoppedTire _vehicleselfreference 0 false)
        (Niagara|SetNiagaraVariable(Bool) _roadinteractionfx "TirePoppedSparksFront" true)
        (LocalVariables|SetTirePopped_Front true))
      (:wheel_b
        (Niagara|SpawnSystemAttached _poppedtirefx _vehicleskeletalmesh "None" (Transformation|GetSocketLocation _vehicleskeletalmesh (Variables|GetWheelSocketB)) "0, 0, 0" "KeepWorldPosition" false true "AutoRelease")
        (Vehicle|Tires|OnPoppedTire _vehicleselfreference 1 false)
        (Niagara|SetNiagaraVariable(Bool) _roadinteractionfx "TirePoppedSparksRear" true)
        (LocalVariables|SetTirePopped_Rear true))
      (:glass_fr_w
        (if (Utilities|Name|NotEqual(Name) _bonenamelocal "")
          (bind _returnvalue (Niagara|SpawnSystemAttached (Variables|GetWindowBreakFX) _vehicleskeletalmesh _bonenamelocal (.location _windowtransformfrw) (.rotation _windowtransformfrw) "KeepRelativeOffset" false true "AutoRelease"))
          (Niagara|SetNiagaraVariable(Vector3) _returnvalue "windowSize" (.scale _windowtransformfrw)))))
    (return (LocalVariables|GetTirePopped_Front) (LocalVariables|GetTirePopped_Rear))))

(fn ExplosionFX
  (params (ExplodeFX :type Niagara%20System%20Object%20Reference) (ActorLocation :type Vector) (SkeletalMesh :type Mesh%20Component%20Object%20Reference) (WreckFX :type Niagara%20System%20Object%20Reference) (ExplosionCameraShakeRadius :type Float%20%28single-precision%29) (MaxZeroHealthTime :type Float%20%28single-precision%29) (ExplosionShake :type Camera%20Shake%20Base%20Class%20Reference) (DestructionMaterial :type Material%20Interface%20Object%20Reference) (ExplodeFXScale :type Float%20%28double-precision%29))
  (bind _skeletalmesh (Variables|GetSkeletalMesh))
  (bind _returnvalue (Niagara|SpawnSystematLocation :SystemTemplate (Variables|GetExplodeFX) :Location (Variables|GetActorLocation) :Scale "1.000000,1.000000,1.000000" :bAutoDestroy true :bAutoActivate true :PoolingMethod "AutoRelease" :bPreCullCheck true))
  (Effects|Components|ParticleSystem|SetFloatParameter _returnvalue "Global Scale" ExplodeFXScale)
  (for (_ _array_index) (Rendering|Material|GetMaterials _skeletalmesh)
    (Rendering|Material|SetMaterial _skeletalmesh _array_index (Variables|GetDestructionMaterial)))
  (bind _returnvalue_2 (Niagara|SpawnSystemAttached (Variables|GetWreckFX) _skeletalmesh "engine" "0, 0, 0" "0, 0, 0" "SnapToTarget" false true "AutoRelease"))
  (bind _output_get (LocalVariables|SetNS_WreckFX _returnvalue_2))
  (Utilities|IsValid _output_get
    (:Is%20Valid
      (Niagara|SetNiagaraVariable(Float) _output_get "Duration" (Variables|GetMaxZeroHealthTime))
      (Camera|PlayWorldCameraShake 0 (Variables|GetExplosionShake) (Variables|GetActorLocation) 1.0 (Variables|GetExplosionCameraShakeRadius) 0.2 true))))

(fn HandleGearChange
  (params (NewGear :type Integer) (RoadInteractionFX :type FXSystem%20Component%20Object%20Reference) (Driver :type Fort%20Player%20Pawn%20Object%20Reference))
  (bind _driver (Variables|GetDriver))
  (bind _returnvalue (Fortnite|GetFortPlayerControllerfromActor _driver))
  (bind _driver_2 (Variables|GetDriver))
  (bind _returnvalue_2 (Fortnite|GetFortPlayerControllerfromActor _driver_2))
  (bind _newgear (Variables|GetNewGear))
  (if (>= _newgear (LocalVariables|GetCurrentGear))
    (Effects|Components|ParticleSystem|SetBoolParameter (Variables|GetRoadInteractionFX) "IsShifting" true)
    (Game|Feedback|ClientStartCameraShake :self _returnvalue :Shake "/Motorcycle/ForceFeedback/B_Motorcycle_Medium_Jolt_CameraShake.B_Motorcycle_Medium_Jolt_CameraShake_C" :Scale 1.0 :PlaySpace "CameraLocal")
    (Game|Feedback|ClientPlayForceFeedback _returnvalue "/Motorcycle/ForceFeedback/FF_Motorcycle_Medium_Bump.FF_Motorcycle_Medium_Bump" "None" false false false)
    (else
      (Game|Feedback|ClientStartCameraShake :self _returnvalue_2 :Shake "/Motorcycle/ForceFeedback/B_Motorcycle_Small_Jolt_CameraShake.B_Motorcycle_Small_Jolt_CameraShake_C" :Scale 1.0 :PlaySpace "CameraLocal")
      (Game|Feedback|ClientPlayForceFeedback _returnvalue_2 "/Motorcycle/ForceFeedback/FF_Motorcycle_Small_Bump.FF_Motorcycle_Small_Bump" "None" false false false)))
  (LocalVariables|SetCurrentGear _newgear))

(fn HandleScrapeDamage
  (params (GE_RimScrapeDamage :type Gameplay%20Effect%20Class%20Reference) (VehicleAbilitySystem :type Ability%20System%20Component%20Object%20Reference) (NotOnGround :type Boolean))
  (if (not (Variables|GetNotOnGround))
    (GameplayEffects|ApplyGameplayEffectToSelf :self (Variables|GetVehicleAbilitySystem) :GameplayEffectClass (Variables|GetGE_RimScrapeDamage) :Level 0.0)))

(fn HandleTireSmoke
  (params (MaxSpeed :type Float%20%28double-precision%29) (ForwardSpeedKmh :type Float%20%28double-precision%29) (PreviousForwardSpeedKmh :type Float%20%28double-precision%29) (GasPedalIsDown :type Boolean) (GetFuel :type Float%20%28double-precision%29) (VehicleIsInAir :type Boolean) (IsVehicleBoosting :type Boolean) (IsTouchingGround :type Boolean) (RoadInteractionFX :type Niagara%20Particle%20System%20Component%20Object%20Reference) (GC_ScreenFX :type Gameplay%20Tag%20Structure) (DriverPawn :type Actor%20Object%20Reference) (SurfaceTypeArrayforFX :type Array%20of%20EPhysicalSurface%20Enums :default "SurfaceType_Default") (SurfaceTypeVehicleOn :type EPhysicalSurface%20Enum%20%28by%20ref%29 :default "SurfaceType_Default") (TirePoppedRear :type Boolean) (HasTraction :type Boolean))
  (bind _surfacetypevehicleon (Variables|GetSurfaceTypeVehicleOn))
  (bind _surfacetypearrayforfx (Variables|GetSurfaceTypeArrayforFX))
  (bind _returnvalue (Utilities|Array|FindItem _surfacetypearrayforfx _surfacetypevehicleon))
  (bind _returnvalue_2 (CallFunction|IsVehicleAccelerating? MaxSpeed ForwardSpeedKmh PreviousForwardSpeedKmh GasPedalIsDown GetFuel VehicleIsInAir))
  (bind _surfacetypevehicleon_2 (Variables|GetSurfaceTypeVehicleOn))
  (bind _returnvalue_3 (Utilities|Enum|BytetoEnumEPhysicalSurface _surfacetypevehicleon_2))
  (bind _istouchingground (Variables|GetIsTouchingGround))
  (bind _isvehicleboosting (Variables|GetIsVehicleBoosting))
  (bind _returnvalue_4 (and _isvehicleboosting _istouchingground))
  (bind _returnvalue_5 (or _returnvalue_2 _returnvalue_4))
  (bind _roadinteractionfx (Variables|GetRoadInteractionFX))
  (bind _roadinteractionfx_2 (Variables|GetRoadInteractionFX))
  (bind _driverpawn (Variables|GetDriverPawn))
  ;[1 1 1] Acceleration Tire Smoke
  (Utilities|FlowControl|Sequence
    (:then_0
      (if (not _returnvalue_5)
        (Utilities|FlowControl|Sequence
          (:then_0
            _)
          (:then_1
            (Niagara|SetNiagaraVariable(Bool) _roadinteractionfx "PeelOut_Wheel_Rear" false)
            (Niagara|SetNiagaraVariable(Bool) _roadinteractionfx "IsPeelOutOnRoad" false)
            (Niagara|SetNiagaraVariable(Bool) _roadinteractionfx "IsPeelOutOnRoadRewind" true)
            (GameplayCue|RemoveGameplayCueOnActor(Looping) (Variables|GetDriverPawn) (Variables|GetGC_ScreenFX) (Ability|GameplayCue|MakeGameplayCueParameters 0.0)))))
      (Utilities|FlowControl|DoOnce false
        (:Completed
          (if (not (Variables|GetTirePoppedRear))
            (Niagara|SetNiagaraVariable(Bool) _roadinteractionfx "PeelOut_Wheel_Rear" true)
            (Utilities|FlowControl|Sequence
              (:then_0
                (Utilities|IsValid _driverpawn
                  (:Is%20Valid
                    (GameplayCue|AddGameplayCueOnActor(Looping) _driverpawn (Variables|GetGC_ScreenFX) (Ability|GameplayCue|MakeGameplayCueParameters 0.0)) ; Dirt bit FX on screen
                  )))
              (:then_2
                (Niagara|SetNiagaraVariable(Bool) _roadinteractionfx "IsPeelOutOnRoad" (or (Utilities|Enum|Equal(Enum) _returnvalue_3 "SurfaceType_Default") (Utilities|Enum|Equal(Enum) _returnvalue_3 "SurfaceType2")))
                (Niagara|SetNiagaraVariable(Bool) _roadinteractionfx "IsPeelOutOnRoadRewind" false)))))))
    (:then_1
      (Development|PrintString 0 (Utilities|String|ToString(Integer) _returnvalue) false false)
      (Effects|Components|ParticleSystem|SetIntParameter _roadinteractionfx "PeelOutColorIndex" _returnvalue)
      (if (and (or (Utilities|Enum|Equal(Enum) _returnvalue_3 "SurfaceType10") (Utilities|Enum|Equal(Enum) _returnvalue_3 "SurfaceType9")) _returnvalue_5)
        _
        (else
          _)))
    (:then_2
      (if VehicleIsInAir
        (Niagara|SetNiagaraVariable(Bool) _roadinteractionfx_2 "IsInAir" true)
        (else
          (Niagara|SetNiagaraVariable(Bool) _roadinteractionfx_2 "IsInAir" false))))))

(fn HideHUD
  (params (DriverPawn :type Actor%20Object%20Reference) (HUDElementTags :type Gameplay%20Tag%20Container%20Structure%20%28by%20ref%29) (bHideElements :type Boolean))
  (HUD|SetHUDElementVisibilitywithContext (Fortnite|GetFortPlayerControllerfromActor DriverPawn) (Variables|GetHUDElementTags) (Variables|GetbHideElements) "None" ""))

(fn IsVehicleAccelerating?
  (params (MaxSpeed :type Float%20%28double-precision%29) (ForwardSpeedKmh :type Float%20%28double-precision%29) (PreviousForwardSpeedKmh :type Float%20%28double-precision%29) (GasPedalIsDown :type Boolean) (GetFuel :type Float%20%28double-precision%29) (VehicleIsInAir :type Boolean) (IsDoingStoppie :type Boolean))
  (return (Math|Boolean|ANDBoolean (Math|Float|InRange(Float) (Variables|GetForwardSpeedKmh) 2.0 (Variables|GetMaxSpeed)) (> (Variables|GetForwardSpeedKmh) (Variables|GetPreviousForwardSpeedKmh)) (Variables|GetGasPedalIsDown) (>= (Variables|GetGetFuel) 0.01) (not (Variables|GetVehicleIsInAir)))))

(fn OnImpactEvent
  (params (Pawn :type Object%20Reference) (HasDriver :type Boolean) (PassengerIsInstigator :type Boolean) (Passengers :type Array%20of%20Fort%20Player%20Pawn%20Object%20References) (LastDriver :type Actor%20Object%20Reference) (DriverPawn :type Actor%20Object%20Reference) (LaunchVelocity :type Vector) (GetAllPassengers :type Array%20of%20Fort%20Player%20Pawn%20Object%20References) (HitPlayerCue :type Gameplay%20Tag%20Structure) (FiendHitCue :type Gameplay%20Tag%20Structure))
  (bind _pawn (Variables|GetPawn))
  (bind _launchvelocity (Variables|GetLaunchVelocity))
  (bind _getallpassengers (Variables|GetGetAllPassengers))
  (bind _asfort_player_pawn_athena (Utilities|Casting|CastToFortPlayerPawnAthena _pawn)
    (:then
      (Pawn|Character|LaunchCharacterJump _asfort_player_pawn_athena _launchvelocity false true (Team|OnSameTeam (Variables|GetDriverPawn) _asfort_player_pawn_athena) true)
      (for _array_element _getallpassengers
        (GameplayCue|ExecuteGameplayCue%28Burst%29Local _array_element (Variables|GetHitPlayerCue) (LocalVariables|GetEmptyParams))
        (bind _output_get (LocalVariables|SetLastHitPlayer _asfort_player_pawn_athena))
        (if (Variables|GetHasDriver)
          (Class|FortPlayerPawnAthena|SetFallInstigator :self _output_get :NewFallInstigator (Game|GetInstigatorController (Variables|GetDriverPawn)))
          (elif (Variables|GetPassengerIsInstigator)
            (Class|FortPlayerPawnAthena|SetFallInstigator :self _output_get :NewFallInstigator (Game|GetInstigatorController (Utilities|Array|Get(acopy) (Variables|GetPassengers))))
            (else
              (Class|FortPlayerPawnAthena|SetFallInstigator :self _output_get :NewFallInstigator (Game|GetInstigatorController (Variables|GetLastDriver)))))))
      (return))
    (:CastFailed
      (bind _asfort_aipawn (Utilities|Casting|CastToFortAIPawn _pawn))
      (Components|Movement|StopMovementImmediately (Variables|/Script/Engine.Character|GetCharacterMovement _asfort_aipawn))
      (Character|LaunchCharacter _asfort_aipawn _launchvelocity true true) ; Launch player based on cart velocity
      (for _array_element_2 _getallpassengers
        (GameplayCue|ExecuteGameplayCue%28Burst%29Local _array_element_2 (Variables|GetFiendHitCue) (LocalVariables|GetEmptyParams))))))

(fn SetupFXSettingsforVehicleType
  (params (RoadInteractionFX :type Niagara%20Particle%20System%20Component%20Object%20Reference) (WheelScaleFront :type Vector) (WheelScaleRear :type Vector) (TireScaleFront :type Vector) (TireScaleRear :type Vector) (WheelWidth :type Float%20%28single-precision%29) (WheelWidthBack :type Float%20%28single-precision%29) (VehicleHasFuel :type Boolean) (ExhaustColor :type Linear%20Color%20Structure) (ExhaustScale :type Float%20%28single-precision%29))
  (bind _roadinteractionfx (Variables|GetRoadInteractionFX))
  (Effects|Components|ParticleSystem|SetFloatParameter _roadinteractionfx "Wheel_Width_Back" (Variables|GetWheelWidthBack))
  (Niagara|SetNiagaraVariable(Bool) _roadinteractionfx "VehicleHasFuel" (Variables|GetVehicleHasFuel)))

(fn SetupRoadInteractionFX
  (params (IsActive :type Boolean) (IsDedicatedServer :type Boolean) (RoadInteractionsFX :type Niagara%20System%20Object%20Reference) (SkeletalMesh :type Scene%20Component%20Object%20Reference) (WheelScaleFront :type Vector) (WheelScaleRear :type Vector) (TireScaleFront :type Vector) (TireScaleRear :type Vector) (WheelWidth :type Float%20%28single-precision%29) (WheelWidthBack :type Float%20%28single-precision%29) (VehicleHasFuel :type Boolean) (ExhaustColor :type Linear%20Color%20Structure) (ExhaustScale :type Float%20%28single-precision%29))
  (if IsDedicatedServer
    _
    (elif (Variables|GetIsActive)
      (bind _returnvalue (Niagara|SpawnSystemAttached (Variables|GetRoadInteractionsFX) (Variables|GetSkeletalMesh) "root" "0, 0, 0" "0.000000,180.000000,0.000000" "KeepRelativeOffset" true true "AutoRelease"))
      (bind _output_get (LocalVariables|SetFX%20Road%20Interaction%20And%20Exhaust _returnvalue))
      (CallFunction|SetupFXSettingsForVehicleType _output_get (Variables|GetWheelScaleFront) (Variables|GetWheelScaleRear) (Variables|GetTireScaleFront) (Variables|GetTireScaleRear) (Variables|GetWheelWidth) (Variables|GetWheelWidthBack) (Variables|GetVehicleHasFuel) (Variables|GetExhaustColor) (Variables|GetExhaustScale))
      (else
        (Components|Activation|Deactivate (LocalVariables|GetFX%20Road%20Interaction%20And%20Exhaust))
        (LocalVariables|SetFX%20Road%20Interaction%20And%20Exhaust 0)))))

(fn SpringCompressionFF
  (params (DriverPlayerControler :type Player%20Controller%20Object%20Reference) (SpringCompressionFF :type Force%20Feedback%20Effect%20Object%20Reference))
  (Utilities|IsValid DriverPlayerControler
    (:Is%20Valid
      (Game|Feedback|ClientPlayForceFeedback DriverPlayerControler (Variables|GetSpringCompressionFF) "None" false false false))))

(fn SurfaceTypeChange
  (params (SurfaceTypeVehicleIsOn :type EPhysicalSurface%20Enum :default "SurfaceType_Default") (RoadInteractionFX :type Niagara%20Particle%20System%20Component%20Object%20Reference) (FortVehicleInterface :type Ability%20System%20Component%20Object%20Reference) (GE_IceSurface :type Gameplay%20Effect%20Class%20Reference))
  (bind _surfacetypevehicleison (Variables|GetSurfaceTypeVehicleIsOn))
  (bind _fortvehicleinterface (Variables|GetFortVehicleInterface))
  (bind _ge_icesurface (Variables|GetGE_IceSurface))
  (switch Utilities|FlowControl|Switch|SwitchonEPhysicalSurface _surfacetypevehicleison
    (:SurfaceType10
      (Niagara|SetNiagaraVariable(Bool) (Variables|GetRoadInteractionFX) "IsOnDirt" (Utilities|Enum|Equal(Enum) _surfacetypevehicleison "SurfaceType10")))
    (:SurfaceType24
      (if (Utilities|Enum|Equal(Enum) _surfacetypevehicleison "SurfaceType24")
        (GameplayEffects|ApplyGameplayEffectToSelf :self _fortvehicleinterface :GameplayEffectClass _ge_icesurface :Level 0.0)
        (else
          (GameplayEffects|RemoveActiveGameplayEffectbySourceEffect _fortvehicleinterface _ge_icesurface _fortvehicleinterface -1))))))
